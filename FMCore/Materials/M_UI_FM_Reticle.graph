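# UI translucent material that draws a small reticle with an optional jitter driven by the Shake parameter.
# A panned cloud noise lookup at a single point gives a time-varying offset, moving two MF_UI_Box shapes by differing channel pairs in opposite signs.
# Their SDFs merge with Min and SmoothStep into a white core, plus a larger, fainter dark box pair as an outline in the alpha mask.
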Material /FMCore/Materials/M_UI_FM_Reticle
graph {
 properties {
  bAutomaticallySetUsageInEditor=false
  blendMode=BLEND_Translucent
  materialDomain=MD_UI
  pixelDepthOffsetMode=PDOM_Legacy
  usedShadingModels=Default Lit
 }
 nodes [
  n1 = Panner(Coordinate=Constant() {r=0.5}) {speedX=1, speedY=2}
  n2 = Multiply(
   A=ConstantBiasScale(None=TextureSample(UVs=@n1) {samplerType=SAMPLERTYPE_LinearColor, texture=/Game/Global/Textures/Default/Noise/testDiffClouds}.RGB) {bias=-0.5, scale=0.06}
   B=ScalarParameter() {parameterName=Shake}
  )
  n3 = Add(A=TextureCoordinate(), B=ComponentMask(None=@n2) {g=true, r=true})
  n4 = Constant() {r=0.01}
  n5 = MaterialFunctionCall(UVs=@n3, Size=@n4) {materialFunction=&MF_UI_Box}
  n6 = Subtract(A=TextureCoordinate(), B=AppendVector(A=ComponentMask(None=@n2) {b=true}, B=ComponentMask(None=@n2) {g=true}))
  n7 = MaterialFunctionCall(UVs=@n6, Size=@n4) {materialFunction=&MF_UI_Box}
  n8 = OneMinus(None=SmoothStep(Value=Min(A=@n5.SDF, B=@n7.SDF)) {constMax=0.015})
  LinearInterpolate(Alpha=@n8)
  n9 = Constant() {r=0.02}
  n10 = SmoothStep(Value=Min(
    A=MaterialFunctionCall(UVs=@n3, Size=@n9) {materialFunction=&MF_UI_Box}.SDF
    B=MaterialFunctionCall(UVs=@n6, Size=@n9) {materialFunction=&MF_UI_Box}.SDF
   )) {constMax=0.015}
  Max(A=@n8, B=Multiply(A=OneMinus(None=@n10)) {constB=0.4})
 ]
}
